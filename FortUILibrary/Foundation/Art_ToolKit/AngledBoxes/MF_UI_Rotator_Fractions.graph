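# UI material function in the Art ToolKit AngledBoxes folder that rotates texture coordinates by a fraction of a full turn.
# The Rotation_Fraction input drives a Rotator's time, with speed set to approximately two pi, so a value of one equals one revolution.
# The rotation pivots on the Rotator's default center, and an unconnected fraction defaults to zero, replacing animated time with a static angle.
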MaterialFunction /FortUILibrary/Foundation/Art_ToolKit/AngledBoxes/MF_UI_Rotator_Fractions
graph {nodes [
  n1 = Constant()
  FunctionOutput(None=Rotator(
    Coordinate=FunctionInput() {inputName=TexCoord, inputType=FunctionInput_Vector2}
    Time=FunctionInput(Preview=@n1) {bUsePreviewValueAsDefault=true, inputName=Rotation_Fraction, inputType=FunctionInput_Scalar}
   ) {speed=6.2832})
 ]}
